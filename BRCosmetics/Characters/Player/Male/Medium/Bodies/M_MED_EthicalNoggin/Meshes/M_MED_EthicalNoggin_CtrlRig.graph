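# Control Rig for the M_MED_EthicalNoggin male body that drives corrective deform and accessory bones from the body pose.
# A sequence runs the shared FN_DeformRig_FNC, then per-area fixes for heels, knees, glutes, scapulae, belts, wrists, elbows and armband.
# Spherical pose readers weight bone offsets, alongside parent, position, aim and rotation constraints.
# isMoving scales thigh-driven belt offsets, isParachuite gates an elbow-bone fix, and the control-to-bone Rig_Control function is unwired.
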ControlRigBlueprint /BRCosmetics/Characters/Player/Male/Medium/Bodies/M_MED_EthicalNoggin/Meshes/M_MED_EthicalNoggin_CtrlRig
graphs [
 New Function_ContainedGraph {nodes [
   n1 = RigVMFunctionEntryNode()
   RigVMFunction_Sequence = &RigVMFunction_Sequence(ExecuteContext=@n1.Argument)
   &RigUnit_SetTransform(
    Item=(Type=Bone,Name="upperarm_r")
    Weight=1
    bInitial=False
    Value=&RigUnit_GetTransform(Item=(Type=Control,Name="upperarm_r_ctrl"), bInitial=False).Transform
    ExecutePin=@RigVMFunction_Sequence.A
   )
   > &RigUnit_SetTransform(
    Item=(Type=Bone,Name="lowerarm_r")
    Weight=1
    bInitial=False
    Value=&RigUnit_GetTransform(Item=(Type=Control,Name="lowerarm_r_ctrl"), bInitial=False).Transform
   )
   > &RigUnit_SetTransform(
    Item=(Type=Bone,Name="lowerarm_twist_01_r")
    Weight=1
    bInitial=False
    Value=&RigUnit_GetTransform(Item=(Type=Control,Name="lowerarm_twist_01_r_ctrl"), bInitial=False).Transform
   )
   RigUnit_GetTransform = &RigUnit_GetTransform(Item=(Type=Control,Name="hand_r_ctrl"), bInitial=False)
   > RigUnit_SetTransform = &RigUnit_SetTransform(Item=(Type=Bone,Name="hand_r"), Weight=1, bInitial=False, Value=@RigUnit_GetTransform.Transform)
   &RigUnit_SetTransform(
    Item=(Type=Bone,Name="thigh_r")
    Weight=1
    bInitial=False
    Value=&RigUnit_GetTransform(Item=(Type=Control,Name="thigh_r_ctrl"), bInitial=False).Transform
    ExecutePin=@RigVMFunction_Sequence.B
   )
   > &RigUnit_SetTransform(
    Item=(Type=Bone,Name="calf_r")
    Weight=1
    bInitial=False
    Value=&RigUnit_GetTransform(Item=(Type=Control,Name="calf_r_ctrl"), bInitial=False).Transform
   )
   > &RigUnit_SetTransform(
    Item=(Type=Bone,Name="foot_r")
    Weight=1
    bInitial=False
    Value=&RigUnit_GetTransform(Item=(Type=Control,Name="foot_r_ctrl"), bInitial=False).Transform
   )
   > n2 = &RigUnit_SetTransform(
    Item=(Type=Bone,Name="ball_r")
    Weight=1
    bInitial=False
    Value=&RigUnit_GetTransform(Item=(Type=Control,Name="ball_r_ctrl"), bInitial=False).Transform
   )
   RigVMFunctionReturnNode(Argument=@n2.ExecutePin)
  ]}
 RigVMFunctionLibrary {nodes [Rig_Control = RigVMCollapseNode(Argument=())]}
 RigVMFunction_Sequence_1_ContainedGraph {nodes [
   n1 = RigVMFunctionEntryNode()
   RigVMFunction_Sequence = &RigVMFunction_Sequence(ExecuteContext=@n1.ExecuteContext)
   n2 = &RigVMFunction_Sequence(ExecuteContext=@RigVMFunction_Sequence.B)
   n3 = &RigVMFunction_Sequence(ExecuteContext=@n2.B)
   n4 = &RigVMFunction_Sequence(ExecuteContext=@n3.B)
   n5 = &RigVMFunction_Sequence(ExecuteContext=@n4.B)
   n6 = &RigVMFunction_Sequence(ExecuteContext=@n5.B)
   n7 = &RigVMFunction_Sequence(ExecuteContext=@n6.B)
   n8 = &RigVMFunction_Sequence(ExecuteContext=@n7.B)
   n9 = &RigVMFunction_Sequence(ExecuteContext=@n8.B)
   n10 = &RigVMFunction_Sequence(ExecuteContext=@n9.B)
   RigVMFunctionReturnNode(A=@RigVMFunction_Sequence.A, B=@n2.A, C=@n3.A, D=@n4.A, E=@n5.A, F=@n6.A, G=@n7.A, H=@n8.A, I=@n9.A, J=@n10.A, K=@n10.B)
  ]}
 RigVMModel {nodes [
   &RigUnit_OffsetTransformForItem(
    ExecutePin=()
    Item=(Type=Bone,Name="clavicle_l")
    OffsetTransform=(Rotation=(X=0,Y=-0.2588,Z=0,W=0.9659),Translation=(X=0,Y=0,Z=0),Scale3D=(X=1,Y=1,Z=1))
    Weight=1
    bPropagateToChildren=True
   )
   > &RigUnit_OffsetTransformForItem(
    Item=(Type=Bone,Name="lowerarm_l")
    OffsetTransform=(Rotation=(X=0,Y=0,Z=0,W=1),Translation=(X=0,Y=0,Z=0),Scale3D=(X=1,Y=1,Z=1))
    Weight=1
    bPropagateToChildren=True
   )
   RigUnit_BeginExecution = &RigUnit_BeginExecution()
   RigVMFunction_Sequence = RigVMAggregateNode(ExecuteContext=@RigUnit_BeginExecution.ExecutePin)
   FN_DeformRig_FNC = RigVMFunctionReferenceNode(
    DeformRig_ Curve="(deformRig_l_upperarm_up_alpha_72_#1=False,deformRig_r_upperarm_up_alpha_73_#2=False,deformRig_l_upperarm_fr_alpha_74_#3=False,deformRig_r_upperarm_fr_alpha_75_CB51B8 ...(truncated)"
    DeformRig_Bool="(upperarm_1_#4=True,elbow_3_#5=True,wrist_7_#6=false,groin_8_#7=True,glute_17_#8=true,knee_9_40C408324CA26 ...(truncated)"
    DeformRig_Debug=(upperarm_debug_18_#4=False,wrist_debug_20_#6=False,lat_debug_26_#9=false,groinGlute_debug_27_#7=False)
    DeformRig_Value="(upperarm_up_18_#10=(Rotation=(X=0,Y=0,Z=0,W=1),Translation=(X=0,Y=0,Z=20),Scale3D=(X=0.7,Y=1,Z=1)),upperarm_fwd_fr_19_#11=(Rotation=(X=0,Y=0,Z=0,W=1),Translation=(X=0,Y=25,Z=0),Scale3D=(X=0.5,Y=1,Z ...(truncated)"
    ExecuteContext=@RigVMFunction_Sequence.A
   )
   > &RigUnit_OffsetTransformForItem(
    Item=(Type=Bone,Name="attach_cape")
    OffsetTransform=(Rotation=(X=0,Y=0,Z=0,W=1),Translation=(X=0,Y=-1.2,Z=0),Scale3D=(X=1,Y=1,Z=1))
    Weight=1
    bPropagateToChildren=True
   )
   n1 = &RigUnit_SphericalPoseReader(
    ActiveRegionScaleFactors=(PositiveWidth=1,NegativeWidth=1,PositiveHeight=1,NegativeHeight=1)
    ActiveRegionSize=0.11
    Debug=(bDrawDebug=false,bDraw2D=False,bDrawLocalAxes=False,DebugScale=10,DebugSegments=20,DebugThickness=0.1)
    DriverAxis=(X=1,Y=0,Z=0)
    DriverItem=(Type=Bone,Name="calf_r")
    FalloffRegionScaleFactors=(PositiveWidth=1.7,NegativeWidth=1,PositiveHeight=1.5,NegativeHeight=1.5)
    FalloffSize=0.5
    FlipHeightScaling=False
    FlipWidthScaling=False
    OptionalParentItem=(Type=Bone,Name="None")
    RotationOffset=(X=90,Y=0,Z=60)
    ExecutePin=@RigVMFunction_Sequence.D
   )
   > &RigUnit_OffsetTransformForItem(
    Item=(Type=Bone,Name="deform_kneePad_out_r")
    OffsetTransform=(Rotation=(X=0,Y=0,Z=0.4226,W=0.9063),Translation=(X=0,Y=-2,Z=0),Scale3D=(X=1,Y=1,Z=1))
    bPropagateToChildren=True
    Weight=@n1.OutputParam
   )
   > n2 = &RigUnit_SphericalPoseReader(
    ActiveRegionScaleFactors=(PositiveWidth=1,NegativeWidth=1,PositiveHeight=1,NegativeHeight=1)
    ActiveRegionSize=0.18
    Debug=(bDrawDebug=false,bDraw2D=False,bDrawLocalAxes=False,DebugScale=10,DebugSegments=20,DebugThickness=0.1)
    DriverAxis=(X=1,Y=0,Z=0)
    DriverItem=(Type=Bone,Name="thigh_r")
    FalloffRegionScaleFactors=(PositiveWidth=1,NegativeWidth=1,PositiveHeight=1,NegativeHeight=1)
    FalloffSize=0.5
    FlipHeightScaling=False
    FlipWidthScaling=False
    OptionalParentItem=(Type=Bone,Name="None")
    RotationOffset=(X=90,Y=0,Z=0)
   )
   > &RigUnit_OffsetTransformForItem(
    Item=(Type=Bone,Name="deform_glute_r")
    OffsetTransform=(Rotation=(X=0,Y=0,Z=0,W=1),Translation=(X=4,Y=12,Z=0),Scale3D=(X=1,Y=1,Z=1))
    bPropagateToChildren=True
    Weight=@n2.OutputParam
   )
   > n3 = &RigUnit_SphericalPoseReader(
    ActiveRegionScaleFactors=(PositiveWidth=1,NegativeWidth=1,PositiveHeight=1,NegativeHeight=1)
    ActiveRegionSize=0.11
    Debug=(bDrawDebug=false,bDraw2D=False,bDrawLocalAxes=False,DebugScale=10,DebugSegments=20,DebugThickness=0.1)
    DriverAxis=(X=-1,Y=0,Z=0)
    DriverItem=(Type=Bone,Name="calf_l")
    FalloffRegionScaleFactors=(PositiveWidth=1,NegativeWidth=1.7,PositiveHeight=1.5,NegativeHeight=1.5)
    FalloffSize=0.5
    FlipHeightScaling=False
    FlipWidthScaling=False
    OptionalParentItem=(Type=Bone,Name="None")
    RotationOffset=(X=-90,Y=0,Z=60)
   )
   > &RigUnit_OffsetTransformForItem(
    Item=(Type=Bone,Name="deform_kneePad_out_l")
    OffsetTransform=(Rotation=(X=0,Y=0,Z=0.4226,W=0.9063),Translation=(X=0,Y=2,Z=0),Scale3D=(X=1,Y=1,Z=1))
    bPropagateToChildren=True
    Weight=@n3.OutputParam
   )
   > n4 = &RigUnit_SphericalPoseReader(
    ActiveRegionScaleFactors=(PositiveWidth=1,NegativeWidth=1,PositiveHeight=1,NegativeHeight=1)
    ActiveRegionSize=0.18
    Debug=(bDrawDebug=false,bDraw2D=False,bDrawLocalAxes=False,DebugScale=10,DebugSegments=20,DebugThickness=0.1)
    DriverAxis=(X=-1,Y=0,Z=0)
    DriverItem=(Type=Bone,Name="thigh_l")
    FalloffRegionScaleFactors=(PositiveWidth=1,NegativeWidth=1,PositiveHeight=1,NegativeHeight=1)
    FalloffSize=0.5
    FlipHeightScaling=False
    FlipWidthScaling=False
    OptionalParentItem=(Type=Bone,Name="None")
    RotationOffset=(X=-90,Y=0,Z=0)
   )
   > &RigUnit_OffsetTransformForItem(
    Item=(Type=Bone,Name="deform_glute_l")
    OffsetTransform=(Rotation=(X=0,Y=0,Z=0,W=1),Translation=(X=-4,Y=-12,Z=0),Scale3D=(X=1,Y=1,Z=1))
    bPropagateToChildren=True
    Weight=@n4.OutputParam
   )
   &RigUnit_ParentConstraint(
    AdvancedSettings=(InterpolationType=Average,RotationOrderForFilter=XZY)
    Child=(Type=Bone,Name="deform_scapula_r")
    Filter=(TranslationFilter=(bX=True,bY=True,bZ=True),RotationFilter=(bX=false,bY=false,bZ=false),ScaleFilter=(bX=false,bY=false,bZ=false))
    Parents="((Item=(Type=Bone,Name=\"spine_03\"),Weight=0.0027),(Item=(Type=Bone,Name=\"spine_04\"),Weight=0.0095),(Item=(Type=Bone,Name=\"spine_05\"),Weight=0.1063),(Item=(Type=Bone,Name=\"clavicle_r\"),Weight=0.6356),(Item=(Type=Bone,Name=\"upperarm_twist_01_r\"),Weight=0.083 ...(truncated)"
    Weight=1
    ExecutePin=@RigVMFunction_Sequence.E
   )
   > n5 = &RigUnit_SphericalPoseReader(
    ActiveRegionScaleFactors=(PositiveWidth=1,NegativeWidth=1,PositiveHeight=1,NegativeHeight=1)
    ActiveRegionSize=0.4
    Debug=(bDrawDebug=false,bDraw2D=False,bDrawLocalAxes=false,DebugScale=20,DebugSegments=20,DebugThickness=0.1)
    DriverAxis=(X=0,Y=1,Z=0)
    DriverItem=(Type=Bone,Name="clavicle_r")
    FalloffRegionScaleFactors=(PositiveWidth=1,NegativeWidth=1,PositiveHeight=1,NegativeHeight=1)
    FalloffSize=0.5
    FlipHeightScaling=False
    FlipWidthScaling=False
    OptionalParentItem=(Type=Bone,Name="spine_04")
    RotationOffset=(X=0,Y=0,Z=0)
   )
   > n6 = &RigUnit_SphericalPoseReader(
    ActiveRegionScaleFactors=(PositiveWidth=1,NegativeWidth=1,PositiveHeight=1,NegativeHeight=0.2)
    ActiveRegionSize=0.4
    Debug=(bDrawDebug=false,bDraw2D=False,bDrawLocalAxes=False,DebugScale=20,DebugSegments=20,DebugThickness=0.1)
    DriverAxis=(X=-1,Y=0,Z=0)
    DriverItem=(Type=Bone,Name="upperarm_r")
    FalloffRegionScaleFactors=(PositiveWidth=1,NegativeWidth=1,PositiveHeight=1,NegativeHeight=0.3)
    FalloffSize=0.5
    FlipHeightScaling=False
    FlipWidthScaling=False
    OptionalParentItem=(Type=Bone,Name="spine_04")
    RotationOffset=(X=0,Y=45,Z=0)
   )
   > &RigUnit_OffsetTransformForItem(
    Item=(Type=Bone,Name="deform_scapula_r")
    OffsetTransform=(Rotation=(X=-0.0057,Y=-0.1304,Z=-0.0432,W=0.9905),Translation=(X=0.7,Y=0,Z=1),Scale3D=(X=1,Y=1,Z=1))
    bPropagateToChildren=True
    Weight=&RigVMFunction_MathDoubleMul(B=@n6.OutputParam, A=@n5.OutputParam).Result
   )
   > &RigUnit_ParentConstraint(
    AdvancedSettings=(InterpolationType=Average,RotationOrderForFilter=XZY)
    Child=(Type=Bone,Name="deform_scapula_l")
    Filter=(TranslationFilter=(bX=True,bY=True,bZ=True),RotationFilter=(bX=false,bY=false,bZ=false),ScaleFilter=(bX=false,bY=false,bZ=false))
    Parents="((Item=(Type=Bone,Name=\"spine_03\"),Weight=0.0011),(Item=(Type=Bone,Name=\"spine_04\"),Weight=0.0041),(Item=(Type=Bone,Name=\"spine_05\"),Weight=0.035),(Item=(Type=Bone,Name=\"clavicle_l\"),Weight=0.6854),(Item=(Type=Bone,Name=\"upperarm_twist_01_l\"),Weight=0.1108 ...(truncated)"
    Weight=1
   )
   > n7 = &RigUnit_SphericalPoseReader(
    ActiveRegionScaleFactors=(PositiveWidth=1,NegativeWidth=1,PositiveHeight=1,NegativeHeight=1)
    ActiveRegionSize=0.32
    Debug=(bDrawDebug=true,bDraw2D=False,bDrawLocalAxes=true,DebugScale=20,DebugSegments=20,DebugThickness=0.1)
    DriverAxis=(X=0,Y=-1,Z=0)
    DriverItem=(Type=Bone,Name="clavicle_l")
    FalloffRegionScaleFactors=(PositiveWidth=1,NegativeWidth=1,PositiveHeight=1,NegativeHeight=1)
    FalloffSize=0.5
    FlipHeightScaling=False
    FlipWidthScaling=False
    OptionalParentItem=(Type=Bone,Name="spine_04")
    RotationOffset=(X=180,Y=0,Z=0)
   )
   > n8 = &RigUnit_SphericalPoseReader(
    ActiveRegionScaleFactors=(PositiveWidth=1,NegativeWidth=1,PositiveHeight=1,NegativeHeight=0.2)
    ActiveRegionSize=0.4
    Debug=(bDrawDebug=false,bDraw2D=False,bDrawLocalAxes=False,DebugScale=20,DebugSegments=20,DebugThickness=0.1)
    DriverAxis=(X=1,Y=0,Z=0)
    DriverItem=(Type=Bone,Name="upperarm_l")
    FalloffRegionScaleFactors=(PositiveWidth=1,NegativeWidth=1,PositiveHeight=1,NegativeHeight=0.3)
    FalloffSize=0.5
    FlipHeightScaling=False
    FlipWidthScaling=False
    OptionalParentItem=(Type=Bone,Name="spine_04")
    RotationOffset=(X=180,Y=45,Z=0)
   )
   > Get Delta Ref = RigVMFunctionReferenceNode(Item=(Type=Bone,Name="clavicle_l"))
   Remap = &RigVMFunction_MathDoubleRemap(SourceMaximum=-4.7, SourceMinimum=0, TargetMaximum=1, TargetMinimum=0, bClamp=False, Value=@"Get Delta Ref".Result.Location.Y)
   > &RigUnit_OffsetTransformForItem(
    Item=(Type=Bone,Name="deform_scapula_l")
    OffsetTransform=(Rotation=(X=0,Y=0,Z=0.0436,W=0.999),Translation=(X=0.7,Y=0,Z=0),Scale3D=(X=1,Y=1,Z=1))
    bPropagateToChildren=True
    Weight=@Remap.Result
   )
   > &RigUnit_OffsetTransformForItem(
    Item=(Type=Bone,Name="deform_scapula_l")
    OffsetTransform=(Rotation=(X=0.0651,Y=-0.2064,Z=0.2936,W=0.9311),Translation=(X=0,Y=4,Z=-2.5),Scale3D=(X=1,Y=1,Z=1))
    bPropagateToChildren=True
    Weight=&RigVMFunction_MathDoubleMul(A=@n8.OutputParam, B=@n7.OutputParam).Result
   )
   > &RigUnit_OffsetTransformForItem(
    Item=(Type=Bone,Name="deform_scapula_l")
    OffsetTransform=(Rotation=(X=0,Y=0,Z=0,W=1),Translation=(X=-0.1,Y=0,Z=0),Scale3D=(X=1,Y=1,Z=1))
    Weight=1
    bPropagateToChildren=True
   )
   n9 = &RigUnit_SphericalPoseReader(
    ActiveRegionScaleFactors=(PositiveWidth=1,NegativeWidth=1,PositiveHeight=3,NegativeHeight=3)
    ActiveRegionSize=0.1
    Debug=(bDrawDebug=false,bDraw2D=False,bDrawLocalAxes=False,DebugScale=25,DebugSegments=20,DebugThickness=0.1)
    DriverAxis=(X=1,Y=0,Z=0)
    DriverItem=(Type=Bone,Name="spine_02")
    FalloffRegionScaleFactors=(PositiveWidth=1,NegativeWidth=1,PositiveHeight=1,NegativeHeight=1)
    FalloffSize=0.5
    FlipHeightScaling=False
    FlipWidthScaling=False
    OptionalParentItem=(Type=Bone,Name="pelvis")
    RotationOffset=(X=90,Y=0,Z=0)
    ExecutePin=@RigVMFunction_Sequence.I
   )
   > &RigUnit_OffsetTransformForItem(
    Item=(Type=Bone,Name="deform_clamp")
    OffsetTransform=(Rotation=(X=0,Y=0,Z=0,W=1),Translation=(X=0,Y=9,Z=-20),Scale3D=(X=1,Y=1,Z=1))
    bPropagateToChildren=True
    Weight=@n9.OutputParam
   )
   > &RigUnit_OffsetTransformForItem(
    Item=(Type=Bone,Name="deform_belt_l")
    OffsetTransform=(Rotation=(X=0,Y=0,Z=0,W=1),Translation=(X=-15,Y=0,Z=0),Scale3D=(X=1,Y=1,Z=1))
    bPropagateToChildren=True
    Weight=@n9.OutputParam
   )
   > &RigUnit_OffsetTransformForItem(
    Item=(Type=Bone,Name="deform_belt_r")
    OffsetTransform=(Rotation=(X=0,Y=0,Z=0,W=1),Translation=(X=15,Y=0,Z=0),Scale3D=(X=1,Y=1,Z=1))
    bPropagateToChildren=True
    Weight=@n9.OutputParam
   )
   n10 = &RigUnit_SphericalPoseReader(
    ActiveRegionScaleFactors=(PositiveWidth=1,NegativeWidth=1,PositiveHeight=1,NegativeHeight=1)
    ActiveRegionSize=0.3
    Debug=(bDrawDebug=false,bDraw2D=False,bDrawLocalAxes=False,DebugScale=25,DebugSegments=20,DebugThickness=0.25)
    DriverAxis=(X=-1,Y=0,Z=0)
    DriverItem=(Type=Bone,Name="upperarm_r")
    FalloffRegionScaleFactors=(PositiveWidth=1,NegativeWidth=1,PositiveHeight=1,NegativeHeight=1)
    FalloffSize=0.5
    FlipHeightScaling=False
    FlipWidthScaling=False
    OptionalParentItem=(Type=Bone,Name="spine_05")
    RotationOffset=(X=180,Y=0,Z=0)
    ExecutePin=@RigVMFunction_Sequence.K
   )
   > n11 = &RigUnit_SphericalPoseReader(
    ActiveRegionScaleFactors=(PositiveWidth=1,NegativeWidth=1,PositiveHeight=1,NegativeHeight=1)
    ActiveRegionSize=0.3
    Debug=(bDrawDebug=false,bDraw2D=False,bDrawLocalAxes=False,DebugScale=25,DebugSegments=20,DebugThickness=0.25)
    DriverAxis=(X=1,Y=0,Z=0)
    DriverItem=(Type=Bone,Name="upperarm_l")
    FalloffRegionScaleFactors=(PositiveWidth=1,NegativeWidth=1,PositiveHeight=1,NegativeHeight=1)
    FalloffSize=0.5
    FlipHeightScaling=False
    FlipWidthScaling=False
    OptionalParentItem=(Type=Bone,Name="spine_05")
    RotationOffset=(X=0,Y=0,Z=0)
   )
   RigVMFunction_ControlFlowBranch = &RigVMFunction_ControlFlowBranch(ExecuteContext=@n11.ExecutePin, Condition=RigVMVariableNode() {variable=isParachuite}.Value)
   n12 = &RigVMFunction_MathDoubleMul(A=@n11.OutputParam, B=@n10.OutputParam)
   &RigUnit_OffsetTransformForItem(
    Item=(Type=Bone,Name="defom_elbowbone_r")
    OffsetTransform=(Rotation=(X=-0.5,Y=0,Z=0,W=0.866),Translation=(X=-1,Y=-7,Z=3),Scale3D=(X=1,Y=1,Z=1))
    bPropagateToChildren=True
    ExecutePin=@RigVMFunction_ControlFlowBranch.True
    Weight=@n12.Result
   )
   > &RigUnit_OffsetTransformForItem(
    Item=(Type=Bone,Name="defom_elbowbone_l")
    OffsetTransform=(Rotation=(X=-0.5,Y=0,Z=0,W=0.866),Translation=(X=1,Y=7,Z=-3),Scale3D=(X=1,Y=1,Z=1))
    bPropagateToChildren=True
    Weight=@n12.Result
   )
   n13 = &RigUnit_SphericalPoseReader(
    ActiveRegionScaleFactors=(PositiveWidth=1,NegativeWidth=1,PositiveHeight=1,NegativeHeight=1)
    ActiveRegionSize=0.2
    Debug=(bDrawDebug=false,bDraw2D=False,bDrawLocalAxes=False,DebugScale=10,DebugSegments=20,DebugThickness=0.05)
    DriverAxis=(X=1,Y=0,Z=0)
    DriverItem=(Type=Bone,Name="foot_r")
    FalloffRegionScaleFactors=(PositiveWidth=1,NegativeWidth=1,PositiveHeight=1,NegativeHeight=1)
    FalloffSize=0.5
    FlipHeightScaling=False
    FlipWidthScaling=False
    OptionalParentItem=(Type=Bone,Name="calf_twist_01_r")
    RotationOffset=(X=-90,Y=0,Z=0)
    ExecutePin=@RigVMFunction_Sequence.C
   )
   > &RigUnit_OffsetTransformForItem(
    Item=(Type=Bone,Name="deform_footbelt_r")
    OffsetTransform=(Rotation=(X=0,Y=0,Z=-0.2588,W=0.9659),Translation=(X=1.5,Y=8,Z=0.3),Scale3D=(X=1,Y=1,Z=1.3))
    bPropagateToChildren=True
    Weight=@n13.OutputParam
   )
   > &RigUnit_OffsetTransformForItem(
    Item=(Type=Bone,Name="deform_heel_r")
    OffsetTransform=(Rotation=(X=0,Y=0,Z=0,W=1),Translation=(X=-4,Y=0,Z=0),Scale3D=(X=1,Y=1,Z=1))
    bPropagateToChildren=True
    Weight=@n13.OutputParam
   )
   > &RigUnit_OffsetTransformForItem(
    Item=(Type=Bone,Name="deform_spur_r")
    OffsetTransform=(Rotation=(X=0,Y=0,Z=0.1305,W=0.9914),Translation=(X=-1,Y=0,Z=0),Scale3D=(X=1,Y=1,Z=1))
    bPropagateToChildren=True
    Weight=@n13.OutputParam
   )
   > n14 = &RigUnit_SphericalPoseReader(
    ActiveRegionScaleFactors=(PositiveWidth=1,NegativeWidth=1,PositiveHeight=1,NegativeHeight=1)
    ActiveRegionSize=0.2
    Debug=(bDrawDebug=false,bDraw2D=False,bDrawLocalAxes=False,DebugScale=10,DebugSegments=20,DebugThickness=0.1)
    DriverAxis=(X=-1,Y=0,Z=0)
    DriverItem=(Type=Bone,Name="foot_l")
    FalloffRegionScaleFactors=(PositiveWidth=1,NegativeWidth=1,PositiveHeight=1,NegativeHeight=1)
    FalloffSize=0.5
    FlipHeightScaling=False
    FlipWidthScaling=False
    OptionalParentItem=(Type=Bone,Name="calf_twist_01_l")
    RotationOffset=(X=90,Y=0,Z=0)
   )
   > &RigUnit_OffsetTransformForItem(
    Item=(Type=Bone,Name="deform_footbelt_l")
    OffsetTransform=(Rotation=(X=0,Y=0,Z=-0.2588,W=0.9659),Translation=(X=-1.5,Y=-8,Z=-0.3),Scale3D=(X=1,Y=1,Z=1.3))
    bPropagateToChildren=True
    Weight=@n14.OutputParam
   )
   > &RigUnit_OffsetTransformForItem(
    Item=(Type=Bone,Name="deform_heel_l")
    OffsetTransform=(Rotation=(X=0,Y=0,Z=0,W=1),Translation=(X=4,Y=0,Z=0),Scale3D=(X=1,Y=1,Z=1))
    bPropagateToChildren=True
    Weight=@n14.OutputParam
   )
   > &RigUnit_OffsetTransformForItem(
    Item=(Type=Bone,Name="deform_spur_l")
    OffsetTransform=(Rotation=(X=0,Y=0,Z=0.1305,W=0.9914),Translation=(X=1,Y=0,Z=0),Scale3D=(X=1,Y=1,Z=1))
    bPropagateToChildren=True
    Weight=@n14.OutputParam
   )
   n15 = &RigUnit_SphericalPoseReader(
    ActiveRegionScaleFactors=(PositiveWidth=1,NegativeWidth=1,PositiveHeight=1,NegativeHeight=1)
    ActiveRegionSize=0.15
    Debug=(bDrawDebug=false,bDraw2D=False,bDrawLocalAxes=False,DebugScale=10,DebugSegments=20,DebugThickness=0.03)
    DriverAxis=(X=-1,Y=0,Z=0)
    DriverItem=(Type=Bone,Name="lowerarm_r")
    FalloffRegionScaleFactors=(PositiveWidth=1,NegativeWidth=1.2,PositiveHeight=1,NegativeHeight=1)
    FalloffSize=0.5
    FlipHeightScaling=False
    FlipWidthScaling=False
    OptionalParentItem=(Type=Bone,Name="upperarm_twist_02_r")
    RotationOffset=(X=-90,Y=0,Z=40)
    ExecutePin=@RigVMFunction_Sequence.H
   )
   > &RigUnit_OffsetTransformForItem(
    Item=(Type=Bone,Name="deform_elbowbelt_r")
    OffsetTransform=(Rotation=(X=0,Y=0.0872,Z=0,W=0.9962),Translation=(X=5,Y=0,Z=0),Scale3D=(X=1,Y=1,Z=1.2))
    bPropagateToChildren=True
    Weight=@n15.OutputParam
   )
   > &RigUnit_OffsetTransformForItem(
    Item=(Type=Bone,Name="defom_elbowbone_r")
    OffsetTransform=(Rotation=(X=0,Y=0,Z=0,W=1),Translation=(X=-8,Y=0,Z=0),Scale3D=(X=1,Y=1,Z=1))
    bPropagateToChildren=True
    Weight=@n15.OutputParam
   )
   > n16 = &RigUnit_SphericalPoseReader(
    ActiveRegionScaleFactors=(PositiveWidth=1,NegativeWidth=1,PositiveHeight=1,NegativeHeight=1)
    ActiveRegionSize=0.15
    Debug=(bDrawDebug=false,bDraw2D=False,bDrawLocalAxes=False,DebugScale=10,DebugSegments=20,DebugThickness=0.05)
    DriverAxis=(X=1,Y=0,Z=0)
    DriverItem=(Type=Bone,Name="lowerarm_l")
    FalloffRegionScaleFactors=(PositiveWidth=1.2,NegativeWidth=1,PositiveHeight=1,NegativeHeight=1)
    FalloffSize=0.5
    FlipHeightScaling=False
    FlipWidthScaling=False
    OptionalParentItem=(Type=Bone,Name="upperarm_twist_02_l")
    RotationOffset=(X=90,Y=0,Z=40)
   )
   > &RigUnit_OffsetTransformForItem(
    Item=(Type=Bone,Name="deform_elbowbelt_l")
    OffsetTransform=(Rotation=(X=0,Y=0.0872,Z=0,W=0.9962),Translation=(X=-9,Y=0,Z=0),Scale3D=(X=1,Y=1,Z=1.2))
    bPropagateToChildren=True
    Weight=@n16.OutputParam
   )
   > &RigUnit_OffsetTransformForItem(
    Item=(Type=Bone,Name="defom_elbowbone_l")
    OffsetTransform=(Rotation=(X=0,Y=0,Z=0,W=1),Translation=(X=8,Y=0,Z=0),Scale3D=(X=1,Y=1,Z=1))
    bPropagateToChildren=True
    Weight=@n16.OutputParam
   )
   &RigUnit_ParentConstraint(
    AdvancedSettings=(InterpolationType=Average,RotationOrderForFilter=XZY)
    Child=(Type=Bone,Name="armBand")
    Filter=(TranslationFilter=(bX=True,bY=True,bZ=True),RotationFilter=(bX=true,bY=true,bZ=true),ScaleFilter=(bX=false,bY=false,bZ=false))
    Parents=((Item=(Type=Bone,Name="clavicle_r"),Weight=0.0004),(Item=(Type=Bone,Name="upperarm_twist_01_r"),Weight=0.5828),(Item=(Type=Bone,Name="upperarm_twist_02_r"),Weight=0.6883))
    Weight=1
    bMaintainOffset=true
    ExecutePin=@RigVMFunction_Sequence.J
   )
   n17 = &RigUnit_SphericalPoseReader(
    ActiveRegionScaleFactors=(PositiveWidth=1,NegativeWidth=1,PositiveHeight=1,NegativeHeight=1)
    ActiveRegionSize=0.17
    Debug=(bDrawDebug=false,bDraw2D=False,bDrawLocalAxes=False,DebugScale=10,DebugSegments=20,DebugThickness=0.1)
    DriverAxis=(X=1,Y=0,Z=0)
    DriverItem=(Type=Bone,Name="foot_r")
    FalloffRegionScaleFactors=(PositiveWidth=1,NegativeWidth=1,PositiveHeight=1,NegativeHeight=1)
    FalloffSize=0.5
    FlipHeightScaling=False
    FlipWidthScaling=False
    OptionalParentItem=(Type=Bone,Name="None")
    RotationOffset=(X=90,Y=0,Z=0)
    ExecutePin=@RigVMFunction_Sequence.B
   )
   > &RigUnit_OffsetTransformForItem(
    Item=(Type=Bone,Name="deform_heel_base_r")
    OffsetTransform=(Rotation=(X=0,Y=0,Z=-0.2588,W=0.9659),Translation=(X=0,Y=0,Z=0),Scale3D=(X=1,Y=1,Z=1))
    bPropagateToChildren=True
    Weight=@n17.OutputParam
   )
   > &RigUnit_OffsetTransformForItem(
    Item=(Type=Bone,Name="deform_heel_r")
    OffsetTransform=(Rotation=(X=0,Y=0,Z=0,W=1),Translation=(X=12,Y=0,Z=0),Scale3D=(X=1,Y=1,Z=1))
    bPropagateToChildren=True
    Weight=@n17.OutputParam
   )
   > &RigUnit_OffsetTransformForItem(
    Item=(Type=Bone,Name="deform_spur_r")
    OffsetTransform=(Rotation=(X=0,Y=0,Z=0.0872,W=0.9962),Translation=(X=3,Y=-3.5,Z=0),Scale3D=(X=1,Y=1,Z=1))
    bPropagateToChildren=True
    Weight=@n17.OutputParam
   )
   > n18 = &RigUnit_SphericalPoseReader(
    ActiveRegionScaleFactors=(PositiveWidth=1,NegativeWidth=1,PositiveHeight=1,NegativeHeight=1)
    ActiveRegionSize=0.17
    Debug=(bDrawDebug=false,bDraw2D=False,bDrawLocalAxes=False,DebugScale=10,DebugSegments=20,DebugThickness=0.1)
    DriverAxis=(X=-1,Y=0,Z=0)
    DriverItem=(Type=Bone,Name="foot_l")
    FalloffRegionScaleFactors=(PositiveWidth=1,NegativeWidth=1,PositiveHeight=1,NegativeHeight=1)
    FalloffSize=0.5
    FlipHeightScaling=False
    FlipWidthScaling=False
    OptionalParentItem=(Type=Bone,Name="None")
    RotationOffset=(X=-90,Y=0,Z=0)
   )
   > &RigUnit_OffsetTransformForItem(
    Item=(Type=Bone,Name="deform_heel_base_l")
    OffsetTransform=(Rotation=(X=0,Y=0,Z=-0.2588,W=0.9659),Translation=(X=0,Y=0,Z=0),Scale3D=(X=1,Y=1,Z=1))
    bPropagateToChildren=True
    Weight=@n18.OutputParam
   )
   > &RigUnit_OffsetTransformForItem(
    Item=(Type=Bone,Name="deform_heel_l")
    OffsetTransform=(Rotation=(X=0,Y=0,Z=0,W=1),Translation=(X=-12,Y=0,Z=0),Scale3D=(X=1,Y=1,Z=1))
    bPropagateToChildren=True
    Weight=@n18.OutputParam
   )
   > &RigUnit_OffsetTransformForItem(
    Item=(Type=Bone,Name="deform_spur_l")
    OffsetTransform=(Rotation=(X=0,Y=0,Z=0.0872,W=0.9962),Translation=(X=-3,Y=3.5,Z=0),Scale3D=(X=1,Y=1,Z=1))
    bPropagateToChildren=True
    Weight=@n18.OutputParam
   )
   n19 = &RigUnit_SphericalPoseReader(
    ActiveRegionScaleFactors=(PositiveWidth=1,NegativeWidth=1,PositiveHeight=1,NegativeHeight=1)
    ActiveRegionSize=0.16
    Debug=(bDrawDebug=false,bDraw2D=False,bDrawLocalAxes=False,DebugScale=25,DebugSegments=20,DebugThickness=0.1)
    DriverAxis=(X=1,Y=0,Z=0)
    DriverItem=(Type=Bone,Name="thigh_r")
    FalloffRegionScaleFactors=(PositiveWidth=1.5,NegativeWidth=1,PositiveHeight=1,NegativeHeight=1)
    FalloffSize=0.5
    FlipHeightScaling=False
    FlipWidthScaling=False
    OptionalParentItem=(Type=Bone,Name="None")
    RotationOffset=(X=-90,Y=0,Z=-45)
    ExecutePin=@RigVMFunction_Sequence.F
   )
   Multiply = &RigVMFunction_MathDoubleMul(A=@n19.OutputParam, B=RigVMVariableNode() {variable=isMoving}.Value)
   > &RigUnit_OffsetTransformForItem(
    Item=(Type=Bone,Name="deform_belt_r")
    OffsetTransform=(Rotation=(X=0,Y=0,Z=0,W=1),Translation=(X=0,Y=-12,Z=0),Scale3D=(X=1,Y=1,Z=1))
    bPropagateToChildren=True
    Weight=@Multiply.Result
   )
   > n20 = &RigUnit_SphericalPoseReader(
    ActiveRegionScaleFactors=(PositiveWidth=1,NegativeWidth=1,PositiveHeight=1,NegativeHeight=1)
    ActiveRegionSize=0.16
    Debug=(bDrawDebug=false,bDraw2D=False,bDrawLocalAxes=False,DebugScale=25,DebugSegments=20,DebugThickness=0.1)
    DriverAxis=(X=-1,Y=0,Z=0)
    DriverItem=(Type=Bone,Name="thigh_l")
    FalloffRegionScaleFactors=(PositiveWidth=1,NegativeWidth=1.5,PositiveHeight=1,NegativeHeight=1)
    FalloffSize=0.5
    FlipHeightScaling=False
    FlipWidthScaling=False
    OptionalParentItem=(Type=Bone,Name="None")
    RotationOffset=(X=90,Y=0,Z=-45)
   )
   > &RigUnit_OffsetTransformForItem(
    Item=(Type=Bone,Name="deform_belt_l")
    OffsetTransform=(Rotation=(X=0,Y=0,Z=0,W=1),Translation=(X=0,Y=12,Z=0),Scale3D=(X=1,Y=1,Z=1))
    bPropagateToChildren=True
    Weight=&RigVMFunction_MathDoubleMul(A=@n20.OutputParam, B=RigVMVariableNode() {variable=isMoving}.Value).Result
   )
   > &RigUnit_SphericalPoseReader(
    ActiveRegionScaleFactors=(PositiveWidth=1.7,NegativeWidth=0.3,PositiveHeight=1.2,NegativeHeight=1.5)
    ActiveRegionSize=0.05
    Debug=(bDrawDebug=false,bDraw2D=False,bDrawLocalAxes=False,DebugScale=25,DebugSegments=20,DebugThickness=0.1)
    DriverAxis=(X=-1,Y=0,Z=0)
    DriverItem=(Type=Bone,Name="thigh_l")
    FalloffRegionScaleFactors=(PositiveWidth=1,NegativeWidth=3.4,PositiveHeight=1.3,NegativeHeight=2.5)
    FalloffSize=0.16
    FlipHeightScaling=False
    FlipWidthScaling=False
    OptionalParentItem=(Type=Bone,Name="None")
    RotationOffset=(X=90,Y=0,Z=-10)
   )
   > n21 = &RigUnit_SphericalPoseReader(
    ActiveRegionScaleFactors=(PositiveWidth=1,NegativeWidth=1,PositiveHeight=1,NegativeHeight=1)
    ActiveRegionSize=0.05
    Debug=(bDrawDebug=false,bDraw2D=False,bDrawLocalAxes=False,DebugScale=25,DebugSegments=20,DebugThickness=0.1)
    DriverAxis=(X=-1,Y=0,Z=0)
    DriverItem=(Type=Bone,Name="thigh_l")
    FalloffRegionScaleFactors=(PositiveWidth=1,NegativeWidth=1,PositiveHeight=1,NegativeHeight=1)
    FalloffSize=0.35
    FlipHeightScaling=False
    FlipWidthScaling=False
    OptionalParentItem=(Type=Bone,Name="None")
    RotationOffset=(X=90,Y=0,Z=-15)
   )
   > &RigUnit_OffsetTransformForItem(
    Item=(Type=Bone,Name="deform_belt_l")
    OffsetTransform=(Rotation=(X=-0.4957,Y=-0.113,Z=-0.0653,W=0.8586),Translation=(X=-8,Y=0,Z=0),Scale3D=(X=1,Y=1,Z=1))
    bPropagateToChildren=True
    Weight=&RigVMFunction_MathDoubleMul(A=@n21.OutputParam, B=RigVMVariableNode() {variable=isMoving}.Value).Result
   )
   > /Script/ControlRig.RigUnit_PositionConstraintLocalSpaceOffset(
    Child=(Type=Bone,Name="deform_clamp")
    Filter=(bX=true,bY=true,bZ=false)
    Parents=((Item=(Type=Bone,Name="deform_belt_r"),Weight=1),(Item=(Type=Bone,Name="deform_belt_l"),Weight=1))
    Weight=1
    bMaintainOffset=True
   ) {is=struct RigVMUnitNode}
   > &RigUnit_AimConstraintLocalSpaceOffset(
    AimAxis=(X=-1,Y=0,Z=0)
    Child=(Type=Bone,Name="deform_clamp")
    Filter=(bX=True,bY=True,bZ=True)
    Parents=((Item=(Type=Bone,Name="deform_belt_l"),Weight=1))
    Weight=1
    WorldUp=(Target=(X=0,Y=0,Z=1),Kind=Direction,Space=(Type=None,Name="None"))
   )
   n22 = &RigUnit_SphericalPoseReader(
    ActiveRegionScaleFactors=(PositiveWidth=1,NegativeWidth=1,PositiveHeight=2.5,NegativeHeight=2.5)
    ActiveRegionSize=0.17
    Debug=(bDrawDebug=false,bDraw2D=False,bDrawLocalAxes=true,DebugScale=10,DebugSegments=20,DebugThickness=0.05)
    DriverAxis=(X=-1,Y=0,Z=0)
    DriverItem=(Type=Bone,Name="hand_r")
    FalloffRegionScaleFactors=(PositiveWidth=1,NegativeWidth=1.7,PositiveHeight=1.7,NegativeHeight=1.7)
    FalloffSize=0.5
    FlipHeightScaling=False
    FlipWidthScaling=False
    OptionalParentItem=(Type=Bone,Name="lowerarm_twist_01_r")
    RotationOffset=(X=-90,Y=0,Z=45)
    ExecutePin=@RigVMFunction_Sequence.G
   )
   > &RigUnit_RotationConstraintLocalSpaceOffset(
    AdvancedSettings=(InterpolationType=Average,RotationOrderForFilter=XZY)
    Child=(Type=Bone,Name="deform_wrist_up_base_r")
    Filter=(bX=false,bY=false,bZ=true)
    Parents=((Item=(Type=Bone,Name="lowerarm_twist_01_r"),Weight=1),(Item=(Type=Bone,Name="hand_r"),Weight=1))
    Weight=1
    bMaintainOffset=true
   )
   > &RigUnit_RotationConstraintLocalSpaceOffset(
    AdvancedSettings=(InterpolationType=Average,RotationOrderForFilter=XZY)
    Child=(Type=Bone,Name="deform_wrist_dn_base_r")
    Filter=(bX=true,bY=true,bZ=True)
    Parents=((Item=(Type=Bone,Name="lowerarm_twist_01_r"),Weight=1),(Item=(Type=Bone,Name="hand_r"),Weight=1))
    Weight=1
   )
   > &RigUnit_OffsetTransformForItem(
    Item=(Type=Bone,Name="deform_wrist_up_r")
    OffsetTransform=(Rotation=(X=0.0436,Y=-0.0755,Z=0.4981,W=0.8627),Translation=(X=-3,Y=-6,Z=0),Scale3D=(X=1,Y=1,Z=1.1))
    bPropagateToChildren=True
    Weight=@n22.OutputParam
   )
   > n23 = &RigUnit_SphericalPoseReader(
    ActiveRegionScaleFactors=(PositiveWidth=1,NegativeWidth=1,PositiveHeight=1,NegativeHeight=1)
    ActiveRegionSize=0.11
    Debug=(bDrawDebug=false,bDraw2D=False,bDrawLocalAxes=False,DebugScale=10,DebugSegments=20,DebugThickness=0.05)
    DriverAxis=(X=-1,Y=0,Z=0)
    DriverItem=(Type=Bone,Name="hand_r")
    FalloffRegionScaleFactors=(PositiveWidth=1,NegativeWidth=1,PositiveHeight=1,NegativeHeight=1)
    FalloffSize=0.5
    FlipHeightScaling=False
    FlipWidthScaling=False
    OptionalParentItem=(Type=Bone,Name="lowerarm_twist_01_r")
    RotationOffset=(X=90,Y=90,Z=0)
   )
   > &RigUnit_OffsetTransformForItem(
    Item=(Type=Bone,Name="deform_wrist_dn_r")
    OffsetTransform=(Rotation=(X=0,Y=0,Z=0,W=1),Translation=(X=0,Y=12,Z=0),Scale3D=(X=0.5,Y=1,Z=1))
    bPropagateToChildren=True
    Weight=@n23.OutputParam
   )
   > n24 = &RigUnit_SphericalPoseReader(
    ActiveRegionScaleFactors=(PositiveWidth=1,NegativeWidth=1,PositiveHeight=2.5,NegativeHeight=2.5)
    ActiveRegionSize=0.17
    Debug=(bDrawDebug=false,bDraw2D=False,bDrawLocalAxes=False,DebugScale=10,DebugSegments=20,DebugThickness=0.05)
    DriverAxis=(X=1,Y=0,Z=0)
    DriverItem=(Type=Bone,Name="hand_l")
    FalloffRegionScaleFactors=(PositiveWidth=1.7,NegativeWidth=1,PositiveHeight=1.7,NegativeHeight=1.7)
    FalloffSize=0.5
    FlipHeightScaling=False
    FlipWidthScaling=False
    OptionalParentItem=(Type=Bone,Name="lowerarm_twist_01_l")
    RotationOffset=(X=90,Y=0,Z=45)
   )
   > &RigUnit_RotationConstraintLocalSpaceOffset(
    AdvancedSettings=(InterpolationType=Average,RotationOrderForFilter=XZY)
    Child=(Type=Bone,Name="deform_wrist_up_base_l")
    Filter=(bX=false,bY=false,bZ=True)
    Parents=((Item=(Type=Bone,Name="lowerarm_twist_01_l"),Weight=1),(Item=(Type=Bone,Name="hand_l"),Weight=1))
    Weight=1
   )
   > &RigUnit_RotationConstraintLocalSpaceOffset(
    AdvancedSettings=(InterpolationType=Average,RotationOrderForFilter=XZY)
    Child=(Type=Bone,Name="deform_wrist_dn_base_l")
    Filter=(bX=True,bY=True,bZ=True)
    Parents=((Item=(Type=Bone,Name="lowerarm_twist_01_l"),Weight=1),(Item=(Type=Bone,Name="hand_l"),Weight=1))
    Weight=1
   )
   > &RigUnit_OffsetTransformForItem(
    Item=(Type=Bone,Name="deform_wrist_up_l")
    OffsetTransform=(Rotation=(X=0.0436,Y=-0.0755,Z=0.4981,W=0.8627),Translation=(X=3,Y=6,Z=0),Scale3D=(X=1,Y=1,Z=1.1))
    bPropagateToChildren=True
    Weight=@n24.OutputParam
   )
   > n25 = &RigUnit_SphericalPoseReader(
    ActiveRegionScaleFactors=(PositiveWidth=1,NegativeWidth=1,PositiveHeight=1,NegativeHeight=1)
    ActiveRegionSize=0.11
    Debug=(bDrawDebug=false,bDraw2D=False,bDrawLocalAxes=False,DebugScale=10,DebugSegments=20,DebugThickness=0.05)
    DriverAxis=(X=1,Y=0,Z=0)
    DriverItem=(Type=Bone,Name="hand_l")
    FalloffRegionScaleFactors=(PositiveWidth=1,NegativeWidth=1,PositiveHeight=1,NegativeHeight=1)
    FalloffSize=0.5
    FlipHeightScaling=False
    FlipWidthScaling=False
    OptionalParentItem=(Type=Bone,Name="lowerarm_twist_01_l")
    RotationOffset=(X=-90,Y=0,Z=0)
   )
   > &RigUnit_OffsetTransformForItem(
    Item=(Type=Bone,Name="deform_wrist_dn_l")
    OffsetTransform=(Rotation=(X=0,Y=0,Z=0,W=1),Translation=(X=0,Y=-12,Z=0),Scale3D=(X=0.5,Y=1,Z=1))
    bPropagateToChildren=True
    Weight=@n25.OutputParam
   )
   &RigUnit_OffsetTransformForItem(
    ExecutePin=()
    Item=(Type=Bone,Name="deform_elbowbelt_l")
    OffsetTransform=(Rotation=(X=0,Y=0,Z=0,W=1),Translation=(X=-28,Y=3.7,Z=0),Scale3D=(X=1,Y=1,Z=1))
    Weight=1
    bPropagateToChildren=True
   )
   &RigUnit_OffsetTransformForItem(
    ExecutePin=()
    Item=(Type=Bone,Name="deform_elbowbelt_r")
    OffsetTransform=(Rotation=(X=0.2241,Y=0.483,Z=-0.1294,W=0.8365),Translation=(X=0,Y=-1,Z=0),Scale3D=(X=1,Y=1,Z=1.2))
    Weight=1
    bPropagateToChildren=True
   )
   Rig_Control = RigVMFunctionReferenceNode(Argument=())
  ]}
]
hierarchy [
 root {loc=<0 0 0>, rot=<0 0 0 1>}
 pelvis {parent=root, loc=<0 -0.7071 93.1191>, rot=<0.0127 0.707 -0.0127 -0.707>}
 spine_01 {parent=pelvis, loc=<6.7494 -0.6388 0>, rot=<0 0 0.034 -0.9994>}
 spine_02 {parent=spine_01, loc=<7.2518 -0.035 0>, rot=<0 0 0.0238 -0.9997>}
 spine_03 {parent=spine_02, loc=<6.0104 0 0>, rot=<0 0 0.0149 -0.9999>}
 spine_04 {parent=spine_03, loc=<8.1024 0 0>, rot=<0 0 -0.0132 -0.9999>}
 spine_05 {parent=spine_04, loc=<10.0146 0 0>, rot=<0 0 -0.0015 -1>}
 attach_backpack {parent=spine_05, loc=<-0.123 -13.473 0>, rot=<-0.0093 -0.707 0.0093 -0.707>}
 neck_01 {parent=spine_05, loc=<11.3138 0.1529 0>, rot=<0 0 -0.2003 -0.9797>}
 neck_02 {parent=neck_01, loc=<4.5627 0 0>, rot=<0 0 0.0504 -0.9987>}
 head {parent=neck_02, loc=<5.3914 0 0>, rot=<0 0 0.115 -0.9934>}
 clavicle_l {parent=spine_05, loc=<4.3575 -0.1466 -2.3223>, rot=<0.0465 -0.7289 -0.009 -0.6829>}
 upperarm_l {parent=clavicle_l, loc=<15.5111 0 0>, rot=<0.0226 -0.3234 -0.0088 -0.9459>}
 lowerarm_l {parent=upperarm_l, loc=<26.6514 0 0>, rot=<0.0391 -0.0099 -0.2453 -0.9686>}
 hand_l {parent=lowerarm_l, loc=<22.3491 0 0>, rot=<-0.7081 -0.0507 0.0975 -0.6975>}
 thumb_01_l {parent=hand_l, loc=<1.3118 -0.3191 -2.459>, rot=<0.6546 -0.4042 0.0664 -0.6354>}
 thumb_02_l {parent=thumb_01_l, loc=<5.9336 -0.0448 0.0401>, rot=<-0.0037 0.0213 0.298 -0.9543>}
 thumb_03_l {parent=thumb_02_l, loc=<2.2223 0.3682 -0.0907>, rot=<0.0118 -0.013 0.0066 -0.9998>}
 index_metacarpal_l {parent=hand_l, loc=<1.6535 0.0183 -2.96>, rot=<0.0158 -0.0124 -0.0173 -0.9996>}
 index_01_l {parent=index_metacarpal_l, loc=<8.3008 0.0357 -0.0524>, rot=<-0.0044 0.0485 0.0924 -0.9945>}
 index_02_l {parent=index_01_l, loc=<3.3587 -0.3004 0.0452>, rot=<-0.002 0.0027 0.3225 -0.9466>}
 index_03_l {parent=index_02_l, loc=<2.5578 -0.123 -0.0252>, rot=<-0.002 0.0067 0.2558 -0.9667>}
 middle_metacarpal_l {parent=hand_l, loc=<2.1856 -0.1725 -0.5984>, rot=<0.006 0.0168 -0.0329 -0.9993>}
 middle_01_l {parent=middle_metacarpal_l, loc=<7.3275 -0.2447 0.0129>, rot=<-0.0208 0.0042 0.0858 -0.9961>}
 middle_02_l {parent=middle_01_l, loc=<4.0123 0.1401 0.1175>, rot=<-0.0218 0.0258 0.3512 -0.9357>}
 middle_03_l {parent=middle_02_l, loc=<2.6077 0.0534 0.1373>, rot=<0.0082 0.0188 0.1708 -0.9851>}
 ring_metacarpal_l {parent=hand_l, loc=<2.2074 -0.3625 1.3638>, rot=<-0.0284 0.0722 -0.0458 -0.9959>}
 ring_01_l {parent=ring_metacarpal_l, loc=<7.0295 -0.6197 -0.1446>, rot=<-0.0489 -0.0252 0.1892 -0.9804>}
 ring_02_l {parent=ring_01_l, loc=<3.6174 0.4647 -0.1027>, rot=<0 0 0.2806 -0.9598>}
 ring_03_l {parent=ring_02_l, loc=<2.4152 0.114 0.0199>, rot=<0 0 0.1189 -0.9929>}
 pinky_metacarpal_l {parent=hand_l, loc=<1.7616 -0.9753 3.3814>, rot=<-0.0397 0.0814 -0.069 -0.9935>}
 pinky_01_l {parent=pinky_metacarpal_l, loc=<7.1709 -0.4831 -0.1094>, rot=<-0.0654 -0.0035 0.2242 -0.9723>}
 pinky_02_l {parent=pinky_01_l, loc=<2.8274 0.0893 -0.1043>, rot=<0.0037 0.0011 0.2854 -0.9584>}
 pinky_03_l {parent=pinky_02_l, loc=<2.2046 0.0307 -0.0078>, rot=<-0.0002 0.0098 0.1445 -0.9895>}
 deform_wrist_dn_base_l {parent=hand_l, loc=<0 -2.2342 0>, rot=<0 0 0 -1>}
 deform_wrist_dn_l {parent=deform_wrist_dn_base_l, loc=<0 0 0>, rot=<0 0 0 1>}
 deform_wrist_up_base_l {parent=hand_l, loc=<0 2.2682 0>, rot=<0 0 0 -1>}
 deform_wrist_up_l {parent=deform_wrist_up_base_l, loc=<0 0 0>, rot=<0 0 0 1>}
 lowerarm_twist_02_l {parent=lowerarm_l, loc=<9.6199 0 0>, rot=<0 0 0 1>}
 lowerarm_twist_01_l {parent=lowerarm_l, loc=<17.8549 0 0>, rot=<0 0 0 1>}
 deform_elbow_in_base_l {parent=lowerarm_l, loc=<0.2112 5.259 0.8274>, rot=<0 0 0 1>}
 deform_elbow_in_l {parent=deform_elbow_in_base_l, loc=<0 0 0>, rot=<0 0 0 1>}
 deform_elbow_out_base_l {parent=lowerarm_l, loc=<0 -4 0>, rot=<0 0 0 1>}
 deform_elbow_out_l {parent=deform_elbow_out_base_l, loc=<0 0 0>, rot=<0 0 0 1>}
 defom_elbowbone_base_l {parent=lowerarm_l, loc=<1.7821 0.1973 6.5163>, rot=<0 0 0 1>}
 defom_elbowbone_l {parent=defom_elbowbone_base_l, loc=<0 0 0>, rot=<0 0 0 -1>}
 upperarm_twist_01_l {parent=upperarm_l, loc=<0.0041 0 0>, rot=<0 0 0 1>}
 deform_upperarm_up_base_l {parent=upperarm_twist_01_l, loc=<-3.2216 -1.1391 3.3311>, rot=<0 0 0 1>}
 deform_upperarm_up_l {parent=deform_upperarm_up_base_l, loc=<0 0 0>, rot=<0 0 0 1>}
 deform_upperarm_bk_base_l {parent=upperarm_twist_01_l, loc=<1.0813 -7.2193 -1.9453>, rot=<0 0 0 1>}
 deform_upperarm_bk_l {parent=deform_upperarm_bk_base_l, loc=<0 0 0>, rot=<0 0 0 1>}
 deform_upperarm_fr_base_l {parent=upperarm_twist_01_l, loc=<-0.3691 7.0506 -2.765>, rot=<0 0 0 1>}
 deform_upperarm_fr_l {parent=deform_upperarm_fr_base_l, loc=<0 0 0>, rot=<0 0 0 1>}
 upperarm_twist_02_l {parent=upperarm_l, loc=<13.7941 0 0>, rot=<0 0 0 1>}
 deform_elbowbelt_base_l {parent=upperarm_twist_02_l, loc=<9.776 5.279 0.0001>, rot=<0 0 -0.0344 -0.9994>}
 deform_elbowbelt_l {parent=deform_elbowbelt_base_l, loc=<0 0 0>, rot=<0 0 0 1>}
 deform_scapula_l_base {parent=clavicle_l, loc=<11.7874 -8.7861 4.4571>, rot=<-0.6433 -0.7645 0.0084 -0.0408>}
 deform_scapula_l {parent=deform_scapula_l_base, loc=<-0.1 0 0>, rot=<0 0 0 1>}
 attach_cape {parent=spine_05, loc=<9.2254 -12.8958 0>, rot=<0.0616 -0.7044 -0.0616 -0.7044>}
 clavicle_r {parent=spine_05, loc=<4.3575 -0.1466 2.3223>, rot=<0.7289 0.0465 -0.6829 0.009>}
 upperarm_r {parent=clavicle_r, loc=<-15.5111 0 0>, rot=<0.0226 -0.3234 -0.0088 -0.9459>}
 lowerarm_r {parent=upperarm_r, loc=<-26.6515 0 0>, rot=<0.0391 -0.0099 -0.2453 -0.9686>}
 hand_r {parent=lowerarm_r, loc=<-22.3492 0 0>, rot=<-0.7081 -0.0507 0.0975 -0.6975>}
 thumb_01_r {parent=hand_r, loc=<-1.3119 0.3192 2.459>, rot=<0.6546 -0.4042 0.0664 -0.6354>}
 thumb_02_r {parent=thumb_01_r, loc=<-5.9335 0.0447 -0.0401>, rot=<-0.0037 0.0213 0.298 -0.9543>}
 thumb_03_r {parent=thumb_02_r, loc=<-2.2223 -0.3682 0.0907>, rot=<0.0118 -0.013 0.0066 -0.9998>}
 index_metacarpal_r {parent=hand_r, loc=<-1.6534 -0.0181 2.9601>, rot=<0.0158 -0.0124 -0.0173 -0.9996>}
 index_01_r {parent=index_metacarpal_r, loc=<-8.3008 -0.0357 0.0524>, rot=<-0.0044 0.0485 0.0924 -0.9945>}
 index_02_r {parent=index_01_r, loc=<-3.3589 0.3003 -0.0453>, rot=<-0.002 0.0027 0.3225 -0.9466>}
 index_03_r {parent=index_02_r, loc=<-2.5578 0.1231 0.0251>, rot=<-0.002 0.0067 0.2558 -0.9667>}
 middle_metacarpal_r {parent=hand_r, loc=<-2.1856 0.1726 0.5984>, rot=<0.006 0.0168 -0.0329 -0.9993>}
 middle_01_r {parent=middle_metacarpal_r, loc=<-7.3273 0.2447 -0.0129>, rot=<-0.0208 0.0042 0.0858 -0.9961>}
 middle_02_r {parent=middle_01_r, loc=<-4.0142 -0.1411 -0.07>, rot=<-0.0217 0.0259 0.3512 -0.9357>}
 middle_03_r {parent=middle_02_r, loc=<-2.6074 -0.0534 -0.1312>, rot=<0.0083 0.0188 0.1708 -0.9851>}
 ring_metacarpal_r {parent=hand_r, loc=<-2.2073 0.3627 -1.3638>, rot=<-0.0284 0.0722 -0.0458 -0.9959>}
 ring_01_r {parent=ring_metacarpal_r, loc=<-7.0297 0.6197 0.1446>, rot=<-0.0489 -0.0252 0.1892 -0.9804>}
 ring_02_r {parent=ring_01_r, loc=<-3.6171 -0.4648 0.1027>, rot=<0 0 0.2806 -0.9598>}
 ring_03_r {parent=ring_02_r, loc=<-2.4153 -0.1141 -0.0199>, rot=<0 0 0.1189 -0.9929>}
 pinky_metacarpal_r {parent=hand_r, loc=<-1.7617 0.9753 -3.3814>, rot=<-0.0397 0.0814 -0.069 -0.9935>}
 pinky_01_r {parent=pinky_metacarpal_r, loc=<-7.1707 0.4832 0.1094>, rot=<-0.0654 -0.0035 0.2242 -0.9723>}
 pinky_02_r {parent=pinky_01_r, loc=<-2.8274 -0.0892 0.1043>, rot=<0.0037 0.0011 0.2854 -0.9584>}
 pinky_03_r {parent=pinky_02_r, loc=<-2.2045 -0.031 0.0078>, rot=<-0.0002 0.0098 0.1445 -0.9895>}
 deform_wrist_dn_base_r {parent=hand_r, loc=<-0.0004 2.2345 -0.0001>, rot=<0 0 0 -1>}
 deform_wrist_dn_r {parent=deform_wrist_dn_base_r, loc=<0 0 0>, rot=<0 0 0 1>}
 deform_wrist_up_base_r {parent=hand_r, loc=<0.0003 -2.2684 0.0001>, rot=<0 0 0 -1>}
 deform_wrist_up_r {parent=deform_wrist_up_base_r, loc=<0 0 0>, rot=<0 0 0 1>}
 lowerarm_twist_02_r {parent=lowerarm_r, loc=<-9.6199 0 0>, rot=<0 0 0 1>}
 lowerarm_twist_01_r {parent=lowerarm_r, loc=<-17.855 0 0>, rot=<0 0 0 1>}
 deform_elbow_in_base_r {parent=lowerarm_r, loc=<-0.211 -5.2591 -0.8275>, rot=<0 0 0 1>}
 deform_elbow_in_r {parent=deform_elbow_in_base_r, loc=<-0.0001 0 0.0002>, rot=<0 0 0 1>}
 deform_elbow_out_base_r {parent=lowerarm_r, loc=<0 4 -0.0001>, rot=<0 0 0 1>}
 deform_elbow_out_r {parent=deform_elbow_out_base_r, loc=<0 0 0>, rot=<0 0 0 1>}
 defom_elbowbone_base_r {parent=lowerarm_r, loc=<-1.7822 -0.1973 -6.5161>, rot=<0 0 0 1>}
 defom_elbowbone_r {parent=defom_elbowbone_base_r, loc=<0 0 0>, rot=<0 0 0 1>}
 upperarm_twist_01_r {parent=upperarm_r, loc=<-0.0041 0 0>, rot=<0 0 0 1>}
 deform_upperarm_bk_base_r {parent=upperarm_twist_01_r, loc=<-1.0814 7.2192 1.9456>, rot=<0 0 0 1>}
 deform_upperarm_bk_r {parent=deform_upperarm_bk_base_r, loc=<0 0 0>, rot=<0 0 0 1>}
 deform_upperarm_fr_base_r {parent=upperarm_twist_01_r, loc=<0.3689 -7.0506 2.7652>, rot=<0 0 0 1>}
 deform_upperarm_fr_r {parent=deform_upperarm_fr_base_r, loc=<0 0 0>, rot=<0 0 0 1>}
 deform_upperarm_up_base_r {parent=upperarm_twist_01_r, loc=<3.2214 1.1391 -3.3308>, rot=<0 0 0 1>}
 deform_upperarm_up_r {parent=deform_upperarm_up_base_r, loc=<0 0 0>, rot=<0 0 0 -1>}
 upperarm_twist_02_r {parent=upperarm_r, loc=<-13.7942 0 0>, rot=<0 0 0 1>}
 armBand {parent=upperarm_twist_02_r, loc=<0.2328 -0.1362 -5.7256>, rot=<0 -0.6901 0 -0.7237>}
 MSTR_armBand_A {parent=armBand, loc=<0 -0.2186 0.9535>, rot=<0.2571 0.1479 0.1744 -0.9389>}
 armBand_A_1 {parent=MSTR_armBand_A, loc=<0 0 0>, rot=<0 0 0 1>}
 armBand_A_2 {parent=armBand_A_1, loc=<2.6261 0 0>, rot=<0 0 0 1>}
 MSTR_armBand_B {parent=armBand, loc=<-0.0884 0.281 -1.1344>, rot=<-0.0059 -0.2522 -0.2311 -0.9397>}
 armBand_B_1 {parent=MSTR_armBand_B, loc=<0 0 0>, rot=<0 0 0 1>}
 armBand_B_2 {parent=armBand_B_1, loc=<3.0273 0 0>, rot=<0 0 0 1>}
 deform_elbowbelt_base_r {parent=upperarm_twist_02_r, loc=<-9.776 -5.279 0>, rot=<0 0 -0.0344 -0.9994>}
 deform_elbowbelt_r {parent=deform_elbowbelt_base_r, loc=<0 0 0>, rot=<0 0 0 1>}
 deform_scapula_r_base {parent=clavicle_r, loc=<-11.7874 8.7861 -4.4569>, rot=<-0.6433 -0.7645 0.0084 -0.0408>}
 deform_scapula_r {parent=deform_scapula_r_base, loc=<0 0 0>, rot=<0 0 0 1>}
 thigh_l {parent=pelvis, loc=<-2.5693 -0.0109 -7.7501>, rot=<0.0012 0.0572 0.0155 -0.9982>}
 calf_l {parent=thigh_l, loc=<-39.0484 0 0>, rot=<0.0006 -0.0161 -0.0391 -0.9991>}
 foot_l {parent=calf_l, loc=<-42.561 0.1085 0.4291>, rot=<0.0418 -0.0455 0.0343 -0.9975>}
 ball_l {parent=foot_l, loc=<-6.1599 13.8264 0.6806>, rot=<0.0001 -0.0001 -0.7063 -0.7079>}
 deform_footbelt_base_l {parent=foot_l, loc=<1.6011 6.3435 0>, rot=<0 0 0.9465 -0.3227>}
 deform_footbelt_l {parent=deform_footbelt_base_l, loc=<0 0 0>, rot=<0 0 0 1>}
 deform_heel_base_l {parent=foot_l, loc=<0.7983 -1.8139 0>, rot=<0 0 -0.7449 -0.6672>}
 deform_heel_l {parent=deform_heel_base_l, loc=<0 0 0>, rot=<0 0 0 1>}
 deform_spur_l {parent=deform_heel_base_l, loc=<0 0 0>, rot=<0 0 0 1>}
 calf_twist_02_l {parent=calf_l, loc=<-18.8064 -0.0967 0.0397>, rot=<0.0237 0.0095 -0.0035 -0.9997>}
 calf_twist_01_l {parent=calf_l, loc=<-35.813 -0.096 0.0396>, rot=<0.0237 0.0095 -0.0035 -0.9997>}
 deform_knee_in_base_l {parent=calf_l, loc=<1.298 -5.2957 -0.0496>, rot=<0 0 0 -1>}
 deform_knee_in_l {parent=deform_knee_in_base_l, loc=<0 0 0>, rot=<0 0 0 1>}
 deform_knee_out_base_l {parent=calf_l, loc=<0 4 0>, rot=<0 0 0 -1>}
 deform_knee_out_l {parent=deform_knee_out_base_l, loc=<0 0 0>, rot=<0 0 0 1>}
 deform_kneePad_out_l {parent=deform_knee_out_base_l, loc=<0.0001 0 0>, rot=<0 0 0 -1>}
 thigh_twist_01_l {parent=thigh_l, loc=<-21.6303 0 0>, rot=<0.0013 0 0 -1>}
 deform_groin_base_l {parent=thigh_l, loc=<-2.1908 8.1142 -1.7442>, rot=<0 0 0 -1>}
 deform_groin_l {parent=deform_groin_base_l, loc=<0 0 0>, rot=<0 0 0 1>}
 deform_thingh_forbone_l {parent=thigh_l, loc=<3.1901 8.8798 -2.0066>, rot=<0.0981 -0.0041 0.0465 -0.9941>}
 thingh_forbone_l {parent=deform_thingh_forbone_l, loc=<0 0 0>, rot=<0 0 0 1>}
 thigh_r {parent=pelvis, loc=<-2.5691 -0.0109 7.7501>, rot=<0.0572 -0.0012 0.9982 0.0155>}
 calf_r {parent=thigh_r, loc=<39.0487 0 0>, rot=<0.0006 -0.0161 -0.0391 -0.9991>}
 foot_r {parent=calf_r, loc=<42.5611 -0.1086 -0.4292>, rot=<0.0418 -0.0455 0.0343 -0.9975>}
 ball_r {parent=foot_r, loc=<6.1598 -13.8264 -0.6806>, rot=<0.0001 -0.0001 -0.7063 -0.7079>}
 deform_footbelt_base_r {parent=foot_r, loc=<-1.6013 -6.3436 0>, rot=<0 0 0.9465 -0.3227>}
 deform_footbelt_r {parent=deform_footbelt_base_r, loc=<0 0 0>, rot=<0 0 0 1>}
 deform_heel_base_r {parent=foot_r, loc=<-0.7985 1.8139 0>, rot=<0 0 -0.7449 -0.6672>}
 deform_heel_r {parent=deform_heel_base_r, loc=<0 0 0>, rot=<0 0 0 1>}
 deform_spur_r {parent=deform_heel_base_r, loc=<0 0 0>, rot=<0 0 0 1>}
 calf_twist_02_r {parent=calf_r, loc=<18.8064 0.0961 -0.0397>, rot=<0.0237 0.0095 -0.0035 -0.9997>}
 calf_twist_01_r {parent=calf_r, loc=<35.8132 0.0959 -0.0397>, rot=<0.0237 0.0095 -0.0035 -0.9997>}
 deform_knee_in_base_r {parent=calf_r, loc=<-1.2981 5.2958 0.0496>, rot=<0 0 0 1>}
 deform_knee_in_r {parent=deform_knee_in_base_r, loc=<0 0 0>, rot=<0 0 0 1>}
 deform_knee_out_base_r {parent=calf_r, loc=<-0.0001 -4 0>, rot=<0 0 0 -1>}
 deform_kneePad_out_r {parent=deform_knee_out_base_r, loc=<0 0 0>, rot=<0 0 0 1>}
 deform_knee_out_r {parent=deform_knee_out_base_r, loc=<0 0 0>, rot=<0 0 0 1>}
 thigh_twist_01_r {parent=thigh_r, loc=<21.6303 0 0>, rot=<0.0013 0 0 -1>}
 deform_groin_base_r {parent=thigh_r, loc=<2.1908 -8.1142 1.7442>, rot=<0 0 0 -1>}
 deform_groin_r {parent=deform_groin_base_r, loc=<0 0 0>, rot=<0 0 0 1>}
 deform_thingh_forbone_r {parent=thigh_r, loc=<-3.1899 -8.8798 2.0066>, rot=<0.0981 -0.0041 0.0465 -0.9941>}
 thingh_forbone_r {parent=deform_thingh_forbone_r, loc=<0 0 0>, rot=<0 0 0 1>}
 deform_glute_base_l {parent=pelvis, loc=<-8.8064 -8.8222 -8.4287>, rot=<0.0275 0.0501 -0.4472 -0.8926>}
 deform_glute_l {parent=deform_glute_base_l, loc=<0 0 0>, rot=<0 0 0 1>}
 deform_glute_base_r {parent=pelvis, loc=<-8.8061 -8.8222 8.4287>, rot=<0.0501 -0.0275 0.8926 -0.4472>}
 deform_glute_r {parent=deform_glute_base_r, loc=<0 0 0>, rot=<0 0 0 1>}
 deform_belt_base_r {parent=pelvis, loc=<6.3624 9.139 7.2058>, rot=<-0.1534 0.1582 -0.6953 -0.6841>}
 deform_belt_r {parent=deform_belt_base_r, loc=<0 0 0>, rot=<0 0 0 1>}
 deform_belt_base_l {parent=pelvis, loc=<6.3624 9.139 -7.2058>, rot=<0.1582 0.1534 0.6841 -0.6953>}
 deform_belt_l {parent=deform_belt_base_l, loc=<0 0 0>, rot=<0 0 0 -1>}
 deform_clamp_base {parent=pelvis, loc=<6.3624 9.139 0>, rot=<0.5035 -0.4965 0.5035 0.4965>}
 deform_clamp {parent=deform_clamp_base, loc=<0 0 0>, rot=<0 0 0 1>}
 upperarm_r_ctrl_null {type=Null, loc=<-17.7788 -4.8444 134.4502>, rot=<-0.9345 -0.014 -0.3555 -0.0134>}
 upperarm_r_ctrl {controlType=EULER_TRANSFORM, parent=upperarm_r_ctrl_null, type=Control, loc=<-75.8206 -9.2359 112.1711>, rot=<0.9345 0.014 0.3555 -0.0134>}
 lowerarm_r_ctrl_null {parent=upperarm_r_ctrl, type=Null, loc=<-26.6515 0 0>, rot=<0.0391 -0.0099 -0.2453 -0.9686>}
 lowerarm_r_ctrl {controlType=EULER_TRANSFORM, parent=lowerarm_r_ctrl_null, type=Control, loc=<0 0 0>, rot=<0 0 0 1>}
 hand_r_ctrl_null {parent=lowerarm_r_ctrl, type=Null, loc=<-22.3492 0 0>, rot=<-0.7081 -0.0507 0.0975 -0.6975>}
 hand_r_ctrl {controlType=EULER_TRANSFORM, parent=hand_r_ctrl_null, type=Control, loc=<0 0 0>, rot=<0 0 0 1>}
 lowerarm_twist_01_r_ctrl {controlType=EULER_TRANSFORM, parent=lowerarm_r_ctrl, type=Control, loc=<0 0 0>, rot=<0 0 0 1>}
 thigh_r_ctrl {controlType=EULER_TRANSFORM, type=Control, loc=<0 0 0>, rot=<0 0 0 1>}
 calf_r_ctrl {controlType=EULER_TRANSFORM, parent=thigh_r_ctrl, type=Control, loc=<0 0 0>, rot=<0 0 0 1>}
 foot_r_ctrl {controlType=EULER_TRANSFORM, parent=calf_r_ctrl, type=Control, loc=<0 0 0>, rot=<0 0 0 1>}
 ball_r_ctrl {controlType=EULER_TRANSFORM, parent=foot_r_ctrl, type=Control, loc=<0 0 0>, rot=<0 0 0 1>}
]
variables [
 isMoving {type=double, default=0}
 isParachuite {type=bool, default=False}
]
