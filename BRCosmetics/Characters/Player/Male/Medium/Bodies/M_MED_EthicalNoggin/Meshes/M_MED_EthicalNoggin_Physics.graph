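# Physics asset for the EthicalNoggin medium male player body, covering two arm band bone chains.
# Each chain has a small sphere on its master bone and two small capsules on the bones below it, linked in sequence by constraints with moderate swing and tight twist limits.
# A larger capsule on the right upper arm twist bone has no constraint, which suggests it serves as a collider for the bands.
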PhysicsAsset /BRCosmetics/Characters/Player/Male/Medium/Bodies/M_MED_EthicalNoggin/Meshes/M_MED_EthicalNoggin_Physics
bodies {
 MSTR_armBand_A [{shapeType=Sphere, radius=1, shapeName=MSTR_armBand_A_sphere}]
 MSTR_armBand_B [{shapeType=Sphere, radius=1, shapeName=MSTR_armBand_B_sphere}]
 armBand_A_1 [{shapeType=Capsule, length=1, radius=1, shapeName=armBand_A_1_capsule, center=<1.6124 0 0>, rotation=<rot 90 0 0>}]
 armBand_A_2 [{shapeType=Capsule, length=1, radius=1, shapeName=armBand_A_2_capsule, center=<1.6124 0 0>, rotation=<rot 90 0 0>}]
 armBand_B_1 [{shapeType=Capsule, length=1, radius=1, shapeName=armBand_B_1_capsule, center=<1.6124 0 0>, rotation=<rot 90 0 0>}]
 armBand_B_2 [{shapeType=Capsule, length=1, radius=1, shapeName=armBand_B_2_capsule, center=<1.6124 0 0>, rotation=<rot 90 0 0>}]
 upperarm_twist_02_r [{shapeType=Capsule, length=10, radius=5, shapeName=upperarm_twist_02_r_capsule, rotation=<rot 90 0 0>}]
}
constraints [
 {bone1Name=armBand_A_1, bone2Name=armBand_A_2, swing1LimitDegrees=40, swing2LimitDegrees=40, twistLimitDegrees=5}
 {bone1Name=armBand_B_1, bone2Name=armBand_B_2, swing1LimitDegrees=40, swing2LimitDegrees=40, twistLimitDegrees=5}
 {bone1Name=MSTR_armBand_B, bone2Name=armBand_B_1, swing1LimitDegrees=40, swing2LimitDegrees=40, twistLimitDegrees=5}
 {bone1Name=MSTR_armBand_A, bone2Name=armBand_A_1, swing1LimitDegrees=40, swing2LimitDegrees=40, twistLimitDegrees=5}
]
