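# Physics asset for the face accessory mesh of the male medium EthicalNoggin player body, covering a closed-cape dynamic bone and the head.
# It has capsule bodies on the MSTR_dyn_closedcape bone, the dyn_closedcape bone and the head.
# One constraint ties dyn_closedcape to its master bone with narrow swing and twist limit angles.
# The larger head capsule has no constraint to the cape bodies.
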PhysicsAsset /BRCosmetics/Characters/Player/Male/Medium/Bodies/M_MED_EthicalNoggin/Meshes/Parts/M_MED_EthicalNoggin_FaceAcc_Physics
bodies {
 MSTR_dyn_closedcape [{shapeType=Capsule, length=1, radius=1, shapeName=MSTR_dyn_closedcape_capsule}]
 dyn_closedcape [{shapeType=Capsule, length=0.403, radius=3.1019, shapeName=dyn_closedcape_capsule, center=<5.695 0.188 0>}]
 head [{shapeType=Capsule, length=3.4701, radius=8.0299, shapeName=head_capsule, center=<6.3719 1.0913 0>, rotation=<rot 90 -144.7356 -144.7356>}]
}
constraints [{bone1Name=dyn_closedcape, bone2Name=MSTR_dyn_closedcape, swing1LimitDegrees=15, swing2LimitDegrees=7, twistLimitDegrees=1}]
